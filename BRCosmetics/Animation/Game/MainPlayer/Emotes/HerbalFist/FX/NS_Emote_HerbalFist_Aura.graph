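# Niagara system for the HerbalFist emote aura: a short-lived, blue-tinted burst of glow, sphere and rays that fades out.
# SpiderSense_Aura spawns four burst groups, writing each to a visibility tag, with mesh and sprite renderers using separately animated sphere, dark-glow and refraction parameters.
# SpiderSense_Aura_Rays adds two burst groups on two sprite renderers with a derived corona size.
# Colors adapt to time of day, and camera offsets pass through a front-end modifier.
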NiagaraSystem /BRCosmetics/Animation/Game/MainPlayer/Emotes/HerbalFist/FX/NS_Emote_HerbalFist_Aura
dependencies {
 usedDataInterfaces [&NiagaraDataInterfaceCurve, &NiagaraDataInterfaceArrayFloat3, &NiagaraDataInterfaceColorCurve]
 usedDynamicInputs [&SimulationPosition]
 usedModules [&SystemState]
}
emitters {
 SpiderSense_Aura {
  emitterUpdateScript [
   &EmitterState()
   &SpawnBurst_Instantaneous(Spawn Group=3, Use Spawn Probability=false, UseLoopCountLimit=false) {module=SpawnBurst_Instantaneous003}
   &SpawnBurst_Instantaneous(Spawn Group=2, Use Spawn Probability=false, UseLoopCountLimit=false) {module=SpawnBurst_Instantaneous002}
   &SpawnBurst_Instantaneous(Spawn Group=1, Use Spawn Probability=false, UseLoopCountLimit=false) {module=SpawnBurst_Instantaneous001}
   &SpawnBurst_Instantaneous(Spawn Count=1, Use Spawn Probability=false)
  ]
  particleSpawnScript [
   &InitializeParticle(
    Color=&FNTimeOfDayColor(
     Night Color=<rgba 0.356 0.5201 0.8438 0.5>
     Morning Color=<rgba 0.356 0.5201 0.8438 0.75>
     Day Color=<rgba 0.356 0.5201 0.8438 0.75>
     Evening Color=<rgba 0.356 0.5201 0.8438 0.75>
    )
    Lifetime=1.893
    Mesh Scale Mode=::NewEnumerator3
    Mesh Uniform Scale=0.35
    Sprite Size Mode=::NewEnumerator3
    Uniform Sprite Size=240.8027
    UsePositionOffset=false
    Write Material Random=true
   )
   &CameraOffset(Camera Offset Amount=&FN_FrontEndFloatModifier(
     Gameplay=&MergeHelper_FloatSwitch(0=-30, 1=-30, 2=-30, 3=-150, Index=&SpawnGroup(), Num=::NewEnumerator2)
     Front End=&MergeHelper_FloatSwitch(0=-30, 1=-30, 2=-30, 3=-150, Index=&SpawnGroup(), Num=::NewEnumerator2)
    ))
   &InitialMeshOrientation(Mesh Orientation Mode=::NewEnumerator2, Orientation Vector=<0 0 1>, Rotation=<0 0 0.79>, Use Rotation Vector=true)
   /SetVariables(
    Particles.Sphere Mesh Scale:Vector3f=&VectorFromFloat(Value=1)
    Particles.SphereColor:LinearColor=&FNTimeOfDayColor(
     Night Color=<rgba 0.5885 0.8221 1 0.5>
     Morning Color=<rgba 0.5885 0.8221 1 0.75>
     Day Color=<rgba 0.5885 0.8221 1 0.75>
     Evening Color=<rgba 0.5885 0.8221 1 0.75>
    )
    Particles.SphereScale=0.35
   ) {isSetParameters=true}
   /SetVariables(Particles.VisibilityTag=&SpawnGroup()) {isSetParameters=true}
   /SetVariables(
    Particles.Dark Glow Color:LinearColor=&FNTimeOfDayColor(
     Night Color=<rgba 0.0234 0.0352 0.1094 0.5>
     Morning Color=<rgba 0.0234 0.0352 0.1094 0.75>
     Day Color=<rgba 0.0234 0.0352 0.1094 1>
     Evening Color=<rgba 0.0234 0.0352 0.1094 0.75>
    )
    Particles.Dark Glow Sprite Size:Vector2f=&Vector2DFromFloat(Value=230)
   ) {isSetParameters=true}
   /SetVariables(Particles.Refract Color:LinearColor=<rgba 1 1 1 0>) {isSetParameters=true}
  ]
  particleUpdateScript [
   &ParticleState()
   &DynamicMaterialParameters(
    Index 0 Param 1=0.8
    Index 0 Param 2=1
    Module.Index0Float0Write=true
    Module.Index0Float1Write=true
    Module.Index0Float2Write=true
    Module.Index0Float3Write=true
   )
   &ScaleColor(
    Linear Color Curve=DataInterface() {propertyValues {
      AlphaCurve {keys [<at 0 0>,<at 0.047 1>,<at 0.8353 1>,<at 1 0>]}
      BlueCurve {keys [<at 0 6>]}
      GreenCurve {keys [<at 0 6>]}
      RedCurve {keys [<at 0 6>]}
     }}
    Scale Mode=::NewEnumerator2
   ) {module=ScaleColor001}
   &ScaleMeshSize(Scale Factor=&VectorFromFloat(Value=&FloatFromCurve(
      FloatCurve=DataInterface() {propertyValues {Curve {keys [<at 0 0>,<at 0.1 1.4>,<at 1 2>]}}}
      Scale Curve=&Multiply_Float(A=1, B=1)
     )))
   /SetVariables(
    Particles.Sphere Mesh Scale:Vector3f=&Multiply_Vector(
     A=$Particles.Initial.Sphere Mesh Scale
     B=&VectorFromFloat(Value=&FloatFromCurve(FloatCurve=DataInterface() {propertyValues {Curve {keys [
           {arriveTangent=4.1785, interpMode=RCIM_Cubic, leaveTangent=4.1785, tangentMode=RCTM_User, time=0, value=0.05}
           {arriveTangentWeight=0.1028, interpMode=RCIM_Cubic, leaveTangentWeight=0.24, tangentMode=RCTM_User, time=0.1, value=1}
           {arriveTangent=-0.0405, interpMode=RCIM_Cubic, leaveTangent=-0.0405, tangentMode=RCTM_User, time=0.5, value=1}
           {arriveTangentWeight=0.2981, interpMode=RCIM_Cubic, leaveTangentWeight=0.0745, tangentMode=RCTM_User, time=0.9, value=0.9999}
           <at 1 0 RCIM_Cubic>
          ]}}}))
    )
    Particles.SphereColor:LinearColor=&MultiplyLinearColor(
     Linear Color=$Particles.Initial.SphereColor
     Scale Factor=&MakeFloat4FromVectorAndFloat(
      "Vector (RGB)"=&VectorFromFloat(Value=&FloatFromCurve(FloatCurve=DataInterface() {propertyValues {Curve {keys [<at 0 10>,<at 0.15 3>,<at 1 1>]}}}))
      "Float (Alpha)"=&FloatFromCurve(
       FloatCurve=DataInterface() {propertyValues {Curve {keys [
           <at 0 0>
           {interpMode=RCIM_Cubic, tangentMode=RCTM_User, time=0.2, value=1}
           {interpMode=RCIM_Cubic, tangentMode=RCTM_User, time=0.8917, value=1}
           <at 1 0 RCIM_Cubic>
          ]}}}
       Scale Curve=5
      )
     )
    )
    Particles.SphereDynamicParam:Vector4f=&MakeVector4(X=1, Y=0, Z=1, W=1)
   ) {isSetParameters=true}
   /SetVariables(Particles.Inner Glow Dyn 1:Vector4f=<3 1 0 0>) {isSetParameters=true}
   /SetVariables(
    Particles.Dark Glow Color:LinearColor=&MultiplyLinearColor(
     Linear Color=$Particles.Initial.Dark Glow Color
     Scale Factor=&Vector4FromCurve(Vector4Curve=DataInterface() {propertyValues {
        AlphaCurve {keys [<at 0 1>,<at 0.8353 1>,<at 1 0>]}
        BlueCurve {keys [<at 0 1>,<at 1 1>]}
        GreenCurve {keys [<at 0 1>,<at 1 1>]}
        RedCurve {keys [<at 0 1>,<at 1 1>]}
       }})
    )
    Particles.Dark Glow Sprite Size:Vector2f=&Multiply_Vector2D(A=$Particles.Initial.Dark Glow Sprite Size, B=&Vector2DFromFloat(Value=1))
    Particles.DarkGlow Dyn 1:Vector4f=<1.2 1 0 0>
   ) {isSetParameters=true}
   /SetVariables(Particles.Refract DynamicMaterialParameter 1:Vector4f=&MakeVector4(
     X=&FloatFromCurve(FloatCurve=DataInterface() {propertyValues {Curve {keys [
          {interpMode=RCIM_Cubic, leaveTangentWeight=0.0333, tangentMode=RCTM_User, time=0, value=1}
          {arriveTangentWeight=0.0333, interpMode=RCIM_Cubic, leaveTangentWeight=0.1333, tangentMode=RCTM_User, time=0.1, value=0.95}
          {arriveTangentWeight=0.1334, interpMode=RCIM_Cubic, leaveTangentWeight=0.1334, tangentMode=RCTM_User, time=0.5, value=0.95}
          {arriveTangentWeight=0.1333, interpMode=RCIM_Cubic, leaveTangentWeight=0.0333, tangentMode=RCTM_User, time=0.9, value=0.95}
          {arriveTangentWeight=0.0333, interpMode=RCIM_Cubic, tangentMode=RCTM_User, time=1, value=1}
         ]}}})
     Y=1
     Z=1
     W=1
    )) {isSetParameters=true}
  ]
  rendererClasses [&NiagaraMeshRendererProperties, &NiagaraSpriteRendererProperties]
  renderers [
   {rendererClass=&NiagaraMeshRendererProperties, rendererIndex=0}
   {rendererClass=&NiagaraSpriteRendererProperties, rendererIndex=1}
   {rendererClass=&NiagaraSpriteRendererProperties, rendererIndex=2}
   {rendererClass=&NiagaraMeshRendererProperties, rendererIndex=3}
  ]
 }
 SpiderSense_Aura_Rays {
  emitterSpawnScript [/SetVariables(Emitter.SortOrder=1, Emitter.SortOrder2=2) {isSetParameters=true}]
  emitterUpdateScript [
   &EmitterState()
   &SpawnBurst_Instantaneous(Use Spawn Probability=false, UseLoopCountLimit=false)
   &SpawnBurst_Instantaneous(Spawn Group=1, Use Spawn Probability=false, UseLoopCountLimit=false) {module=SpawnBurst_Instantaneous002}
  ]
  particleSpawnScript [
   /SetVariables(Particles.VisibilityTag=&SpawnGroup()) {isSetParameters=true}
   &InitializeParticle(
    Color=<rgba 0.7445 0.7469 1 1>
    Lifetime=1.893
    Mass Mode=::NewEnumerator1
    Mesh Scale Mode=::NewEnumerator3
    Mesh Uniform Scale=&FloatFromCurve(
     FloatCurve=DataInterface() {propertyValues {Curve {keys [{tangentMode=RCTM_User, time=0, value=1}, {tangentMode=RCTM_User, time=1, value=2}]}}}
     CurveIndex=&ReturnNormalizedExecIndex()
     Scale Curve=5
    )
    Sprite Rotation Mode=::NewEnumerator2
    Sprite Size Mode=::NewEnumerator3
    Uniform Sprite Size=220
    UsePositionOffset=false
    Write Material Random=true
   ) {module=InitializeParticle001}
   &CameraOffset(Camera Offset Amount=&FN_FrontEndFloatModifier(Gameplay=-40, Front End=-40))
   &InitialMeshOrientation(
    Mesh Orientation Mode=::NewEnumerator2
    Rotation=&SelectVectorFromArray(
     Direct Array Index=&SpawnGroup()
     Vector Selection Array=DataInterface() {propertyValues {FloatData [<-0.25 0 0>, <0.5 0 0>]}}
    )
    Rotation Coordinate Space=::NewEnumerator2
    Use Rotation Vector=true
   )
  ]
  particleUpdateScript [
   &ParticleState()
   &SolveForcesAndVelocity(Clamp Velocity=false, Limit Acceleration=false)
   &ScaleColor(
    Scale Alpha=&FloatFromCurve(
     FloatCurve=DataInterface() {propertyValues {Curve {keys [
         {arriveTangent=4.0997, interpMode=RCIM_Cubic, leaveTangent=4.0997, tangentMode=RCTM_User, time=0, value=0}
         {interpMode=RCIM_Cubic, tangentMode=RCTM_User, time=0.3, value=1}
         {arriveTangent=-1.6432, interpMode=RCIM_Cubic, leaveTangent=-1.6432, time=0.8041, value=0.9116}
         {arriveTangent=-3.5577, tangentMode=RCTM_User, time=0.9086, value=0}
        ]}}}
     Scale Curve=0.4
    )
    Scale RGB=&MakeVectorFromLinearColorRGB(Color=&FNTimeOfDayColor(Night Color=<rgba 0.225 0.905 1.5 1>, Morning Color=<rgba 1 1 1 1>, Day Color=<rgba 1 1 1 1>, Evening Color=<rgba 1 1 1 1>))
    ScaleA=true
    ScaleRGB=true
   )
   &ScaleSpriteSize(Scale Factor=&Multiply_Vector2D_ByFloat(
     Float=&FloatFromCurve(
      FloatCurve=DataInterface() {propertyValues {Curve {keys [
          {arriveTangent=26.1988, interpMode=RCIM_Cubic, leaveTangent=26.1988, tangentMode=RCTM_User, time=0, value=0}
          {arriveTangent=0.4194, interpMode=RCIM_Cubic, leaveTangent=0.4194, tangentMode=RCTM_User, time=0.0664, value=1}
          {interpMode=RCIM_Cubic, tangentMode=RCTM_User, time=0.5, value=1}
          {arriveTangent=-1.5738, interpMode=RCIM_Cubic, leaveTangent=-1.5738, tangentMode=RCTM_User, time=1, value=0.8}
         ]}}}
      Scale Curve=1.25
     )
     Vector2D=<1 1>
    ))
   /SetVariables(Particles.Corona_Size:Vector2f=&Multiply_Vector2D_ByFloat(Float=0.7, Vector2D=$Particles.SpriteSize)) {isSetParameters=true}
   &DynamicMaterialParameters(
    Index 0 Param 1=10
    Index 0 Param 2=1
    Index 0 Param 3=1
    Index 0 Param 4=1
    Module.Index0Float0Write=true
    Module.Index0Float1Write=true
    Module.Index0Float2Write=true
    Module.Index0Float3Write=true
   )
  ]
  rendererClasses [&NiagaraSpriteRendererProperties]
  renderers [
   {rendererClass=&NiagaraSpriteRendererProperties}
   {rendererClass=&NiagaraSpriteRendererProperties, rendererIndex=1}
  ]
 }
}
summary {
 systemName=NS_Emote_HerbalFist_Aura
 emitters [
  {emitterName=SpiderSense_Aura_Rays, rendererClasses [&NiagaraSpriteRendererProperties]}
  {
   emitterName=SpiderSense_Aura
   rendererClasses [&NiagaraMeshRendererProperties, &NiagaraSpriteRendererProperties]
  }
 ]
 userVariables [
  User.bIsFrontEnd {type=NiagaraBool}
  User.bIsFrontEndPreview {type=NiagaraBool}
 ]
}
