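# Niagara system for the HerbalFist emote's spider-sense effect, drawing pulsing pale-blue meshes at the character's head.
# A single emitter, SpiderSense_Head, continuously spawns oriented mesh particles with a slight camera offset and re-samples the FX_Head skeletal socket every update.
# Curves grow each mesh over its life while fading alpha in and out.
# Custom Sphere attributes hold scale, color and dynamic parameters alongside a dynamic material parameter module.
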NiagaraSystem /BRCosmetics/Animation/Game/MainPlayer/Emotes/HerbalFist/FX/NS_Emote_HerbalFist_SpiderSense
dependencies {
 usedDataInterfaces [&NiagaraDataInterfaceSkeletalMesh, &NiagaraDataInterfaceCurve]
 usedDynamicInputs []
 usedModules [&SystemState]
}
emitters {SpiderSense_Head {
  emitterUpdateScript [
   &EmitterState()
   &SpawnRate(SpawnRate=3, Use Spawn Probability=false)
  ]
  particleSpawnScript [
   &InitializeParticle(Color Mode=::NewEnumerator0, UsePositionOffset=false, Write Material Random=true)
   /SetVariables(
    Particles.Sphere Mesh Scale:Vector3f=&VectorFromFloat(Value=0.3)
    Particles.SphereColor:LinearColor=<rgba 2.2859 2.8615 3.3 1>
    Particles.SphereScale=0.35
   ) {isSetParameters=true}
   &InitialMeshOrientation(
    Mesh Axis To Orient=<0 1 0>
    Mesh Orientation Mode=::NewEnumerator2
    Orientation Vector=<0 0 1>
    Rotation=<0 0 0.5>
    Use Rotation Vector=true
   )
   &CameraOffset(Camera Offset Amount=5)
  ]
  particleUpdateScript [
   &ParticleState()
   &SkeletalMeshLocation(
    MaskSpawnGroup=false
    Mesh Sampling Type=::NewEnumerator5
    Skeletal Mesh=DataInterface() {propertyValues {
      DefaultMesh=null
      ExcludeBoneName=root
      OptionalComponentClass=null
      PreviewMesh=/Game/Characters/Player/Male/Male_Avg_Base/Fortnite_M_Avg_Player
      SkinningMode=SkinOnTheFly
      SoftSourceActor=null
      SourceMode=Default
      UvSetIndex=0
      WholeMeshLOD=-1
      bExcludeBone=false
      bReadDeformedGeometry=true
      bRequireCurrentFrameData=false
      ComponentTags []
      FilteredBones []
      FilteredSockets [FX_Head]
      MeshUserParameter {parameter {name=None, typeDefHandle {registeredTypeIndex=1}, varData []}}
      SamplingRegions []
     }}
    Socket Sampling Mode=::NewEnumerator5
   )
   &DynamicMaterialParameters(
    Index 0 Param 1=1
    Index 0 Param 4=1
    Module.Index0Float0Write=true
    Module.Index0Float1Write=true
    Module.Index0Float2Write=true
    Module.Index0Float3Write=true
   )
   /SetVariables(
    Particles.Sphere Mesh Scale:Vector3f=&Multiply_Vector(
     A=$Particles.Initial.Sphere Mesh Scale
     B=&VectorFromFloat(Value=&FloatFromCurve(FloatCurve=DataInterface() {propertyValues {Curve {keys [
           {interpMode=RCIM_Cubic, leaveTangentWeight=1.4322, tangentMode=RCTM_User, time=0, value=0.5}
           {arriveTangentWeight=0.3336, interpMode=RCIM_Cubic, tangentMode=RCTM_User, time=1, value=1}
          ]}}}))
    )
    Particles.SphereColor:LinearColor=&MultiplyLinearColor(
     Linear Color=$Particles.Initial.SphereColor
     Scale Factor=&MakeFloat4FromVectorAndFloat(
      "Vector (RGB)"=<1 1 1>
      "Float (Alpha)"=&FloatFromCurve(
       FloatCurve=DataInterface() {propertyValues {Curve {keys [<at 0 0 RCIM_Cubic>, {interpMode=RCIM_Cubic, tangentMode=RCTM_User, time=0.5, value=1}, <at 1 0 RCIM_Cubic>]}}}
       Scale Curve=1.5
      )
     )
    )
    Particles.SphereDynamicParam:Vector4f=&MakeVector4(X=1, Y=0, Z=1, W=1)
   ) {isSetParameters=true}
  ]
  rendererClasses [&NiagaraMeshRendererProperties]
  renderers [{rendererClass=&NiagaraMeshRendererProperties, rendererIndex=0}]
 }}
summary {
 systemName=NS_Emote_HerbalFist_SpiderSense
 emitters [{emitterName=SpiderSense_Head, rendererClasses [&NiagaraMeshRendererProperties]}]
 userVariables [
  User.bIsFrontEnd {type=NiagaraBool}
  User.bIsFrontEndPreview {type=NiagaraBool}
 ]
}
