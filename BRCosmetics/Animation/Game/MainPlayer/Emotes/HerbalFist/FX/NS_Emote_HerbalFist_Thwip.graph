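# Niagara effect for the HerbalFist emote's "Thwip": a brief muzzle flash and a shot web strand.
# MuzzleFlash bursts one mesh particle that scales up as it fades.
# Web links spawn-grouped particles into ribbons, lerping each point toward per-group upward offsets over curves and timeline.
# Those points serve as pendulum pivots with gravity and noise for wobble, while curves drive width and fading, and color follows time of day.
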NiagaraSystem /BRCosmetics/Animation/Game/MainPlayer/Emotes/HerbalFist/FX/NS_Emote_HerbalFist_Thwip
dependencies {
 usedDataInterfaces [
  &NiagaraDataInterfaceArrayFloat
  &NiagaraDataInterfaceArrayFloat3
  &NiagaraDataInterfaceCurve
  &NiagaraDataInterfaceColorCurve
  &NiagaraDataInterfaceVectorCurve
 ]
 usedDynamicInputs []
 usedModules [&SystemState, &Timeline]
}
emitters {
 MuzzleFlash {
  emitterUpdateScript [
   &EmitterState()
   &SpawnBurst_Instantaneous(Spawn Count=1, Spawn Time=0.015, Use Spawn Probability=false)
  ]
  particleSpawnScript [
   &InitializeParticle(
    Color=<rgba 0.3177 0.3177 0.3177 0.05>
    Lifetime=0.1
    Mass Mode=::NewEnumerator1
    Mesh Scale Mode=::NewEnumerator3
    Mesh Uniform Scale=0.35
    Position Offset=<0 0 2.5>
    UsePositionOffset=true
    Write Material Random=true
   ) {module=InitializeParticle001}
   &CameraOffset()
   &InitialMeshOrientation(Mesh Orientation Mode=::NewEnumerator2, Orientation Vector=<0 0 1>, Use Rotation Vector=false)
  ]
  particleUpdateScript [
   &ParticleState()
   &Color(
    Color=<rgba 0.6042 0.6042 0.6042 1>
    Scale Alpha=&FloatFromCurve(FloatCurve=DataInterface() {propertyValues {Curve {keys [
         {interpMode=RCIM_Cubic, leaveTangentWeight=0.3334, tangentMode=RCTM_User, time=0, value=1}
         {arriveTangentWeight=1.2624, interpMode=RCIM_Cubic, tangentMode=RCTM_User, time=1, value=0}
        ]}}})
   )
   &ScaleMeshSize(Scale Factor=&VectorFromCurve(VectorCurve=DataInterface() {propertyValues {
       XCurve {keys [
         {arriveTangent=4.1785, interpMode=RCIM_Cubic, leaveTangent=4.1785, tangentMode=RCTM_User, time=0, value=0}
         {arriveTangent=0.9995, interpMode=RCIM_Cubic, leaveTangent=0.9995, tangentMode=RCTM_User, time=0.2, value=0.75}
         {arriveTangent=-0.0405, interpMode=RCIM_Cubic, leaveTangent=-0.0405, tangentMode=RCTM_User, time=1, value=1}
        ]}
       YCurve {keys [
         {arriveTangent=3.1409, interpMode=RCIM_Cubic, leaveTangent=3.1409, tangentMode=RCTM_User, time=0, value=0}
         {interpMode=RCIM_Cubic, tangentMode=RCTM_User, time=0.7, value=1}
        ]}
       ZCurve {keys [
         {arriveTangent=2.9618, interpMode=RCIM_Cubic, leaveTangent=2.9618, tangentMode=RCTM_User, time=0, value=0}
         {interpMode=RCIM_Cubic, tangentMode=RCTM_User, time=0.7, value=1}
        ]}
      }}))
  ]
  rendererClasses [&NiagaraMeshRendererProperties]
  renderers [{rendererClass=&NiagaraMeshRendererProperties, rendererIndex=0}]
 }
 Web {
  emitterUpdateScript [
   &EmitterState()
   &SpawnBurst_Instantaneous(Spawn Count=5, Use Spawn Probability=false, UseLoopCountLimit=false)
  ]
  particleSpawnScript [
   &InitializeRibbon(
    Color=&FNTimeOfDayColor(Night Color=<rgba 2 2 2 1>, Morning Color=<rgba 1.5 1.5 1.5 1>, Day Color=<rgba 1 1 1 1>, Evening Color=<rgba 1.5 1.5 1.5 1>)
    Ribbon Width=2
    UsePositionOffset=false
    Write Material Random=true
   )
   &FN_SetSpawnGroup(Inclusivity=true, Normalize Spawn Group=true, Particles Per Group=5, Pattern=::NewEnumerator0, Set Ribbon ID and Ribbon Link Order=true)
  ]
  particleUpdateScript [
   &ParticleState()
   &DynamicMaterialParameters(
    Index 0 Param 1=1
    Index 0 Param 2=&SelectFloatFromArray(
     Direct Array Index=&SpawnGroup()
     Float Selection Array=DataInterface() {propertyValues {FloatData [0.44,0.66,0.88]}}
    )
    Module.Index0Float0Write=true
    Module.Index0Float1Write=true
    Module.Index0Float2Write=true
    Module.Index0Float3Write=true
   ) {module=DynamicMaterialParameters001}
   /SetVariables(Particles.PositionOrigin:Vector3f=&Lerp_Vector(
     A=&Lerp_Vector(
      A=/Niagara/DynamicInputs/Helpers/SimulationPosition()
      B=&Add_Vector(
       A=&SimulationPosition()
       B=&SelectVectorFromArray(
        Direct Array Index=$Particles.SpawnGroup
        Vector Selection Array=DataInterface() {propertyValues {FloatData [<0 0 200>, <15 -25 170>, <-25 0 135>]}}
       )
      )
      Alpha=&FloatFromCurve(FloatCurve=DataInterface() {propertyValues {Curve {keys [<at 0.12 0 RCIM_Cubic>,<at 0.325 1 RCIM_Cubic>]}}})
     )
     B=&Lerp_Vector(
      A=&SimulationPosition()
      B=&Add_Vector(
       A=&SimulationPosition()
       B=&SelectVectorFromArray(
        Direct Array Index=$Particles.SpawnGroup
        Vector Selection Array=DataInterface() {propertyValues {FloatData [<0 0 200>, <15 -25 190>, <-25 0 155>]}}
       )
      )
      Alpha=$System.Timeline.NormalizedTime
     )
     Alpha=$Particles.Normalized Spawn Group
    )) {isSetParameters=true}
   &GravityForce()
   &VectorNoiseForce(
    Force Amount=&Lerp_Float(
     A=50000
     B=1000
     Alpha=&FloatFromCurve(FloatCurve=DataInterface() {propertyValues {Curve {keys [<at 0 -1>,<at 1 1>]}}}, CurveIndex=$System.Timeline.NormalizedTime)
    )
    Module.Use Random Noise Time Interval=false
   )
   &PendulumSetup(
    Module.Enable Spring Driver=false
    Pendulum Length:NiagaraFloat=&FloatFromCurve(
     FloatCurve=DataInterface() {propertyValues {Curve {keys [
         {interpMode=RCIM_Cubic, leaveTangentWeight=1.1173, tangentMode=RCTM_User, time=0, value=0}
         {arriveTangent=0.2058, interpMode=RCIM_Cubic, leaveTangent=0.2058, tangentMode=RCTM_User, time=0.7131, value=1}
         {arriveTangent=-3.9125, arriveTangentWeight=0.662, interpMode=RCIM_Cubic, leaveTangent=-3.9125, tangentMode=RCTM_User, time=1, value=0}
        ]}}}
     CurveIndex=$Particles.Normalized Spawn Group
     Scale Curve=&FloatFromCurve(
      FloatCurve=DataInterface() {propertyValues {Curve {keys [
          {interpMode=RCIM_Cubic, leaveTangentWeight=1.3766, tangentMode=RCTM_User, time=0.0996, value=0.0194}
          {arriveTangent=-0.1291, interpMode=RCIM_Cubic, leaveTangent=-0.1291, time=0.1591, value=1}
          <at 0.25 0 RCIM_Cubic>
         ]}}}
      Scale Curve=50
     )
    )
    Pendulum Pivot:Vector3f=$Particles.PositionOrigin
    Pendulum Pivot Offset Coordinate Space:ENiagaraCoordinateSpace=::Simulation
    Rigid Pendulum=false
   )
   &SolveForcesAndVelocity(Clamp Velocity=false, Limit Acceleration=false)
   &PendulumConstraint(Calculate and Output Potential Energy=true)
   &CalculateAccurateVelocity()
   &ScaleRibbonWidth(Ribbon Width Scale=&FloatFromCurve(
     FloatCurve=DataInterface() {propertyValues {Curve {keys [
         <at 0 0 RCIM_Cubic>
         {arriveTangent=2, interpMode=RCIM_Cubic, leaveTangent=2, time=0.2, value=1}
         {arriveTangentWeight=0.5797, interpMode=RCIM_Cubic, tangentMode=RCTM_User, time=1, value=2}
        ]}}}
     CurveIndex=&Frac_Float(Float=$Particles.Normalized Spawn Group)
     Scale Curve=$System.Timeline.NormalizedTime
    ))
   &ScaleColor(
    Scale Alpha=&FloatFromCurve(
     FloatCurve=DataInterface() {propertyValues {Curve {keys [
         {interpMode=RCIM_Cubic, tangentMode=RCTM_User, time=0.86, value=1}
         {arriveTangent=-3.3773, interpMode=RCIM_Cubic, leaveTangent=-3.3773, tangentMode=RCTM_User, time=0.99, value=0}
        ]}}}
     CurveIndex=&Lerp_Float(
      A=0
      B=$Particles.Normalized Spawn Group
      Alpha=&FloatFromCurve(FloatCurve=DataInterface() {propertyValues {Curve {keys [<at 0 0 RCIM_Cubic>,<at 0.09 1 RCIM_Cubic>]}}})
     )
    )
    ScaleA=true
    ScaleRGB=true
   )
   &ScaleColor(
    Scale Alpha=&FloatFromCurve(FloatCurve=DataInterface() {propertyValues {Curve {keys [
         {interpMode=RCIM_Cubic, leaveTangentWeight=0.2021, tangentMode=RCTM_User, time=0, value=1}
         {interpMode=RCIM_Cubic, tangentMode=RCTM_User, time=0.22, value=1}
         {arriveTangent=-3.3773, interpMode=RCIM_Cubic, leaveTangent=-3.3773, tangentMode=RCTM_User, time=0.33, value=0}
        ]}}})
    ScaleA=true
    ScaleRGB=true
   ) {module=ScaleColor001}
  ]
  rendererClasses [&NiagaraRibbonRendererProperties]
  renderers [{rendererClass=&NiagaraRibbonRendererProperties}]
 }
}
summary {
 systemName=NS_Emote_HerbalFist_Thwip
 emitters [
  {emitterName=Web, rendererClasses [&NiagaraRibbonRendererProperties]}
  {emitterName=MuzzleFlash, rendererClasses [&NiagaraMeshRendererProperties]}
 ]
 userVariables [User.Player {
   type=NiagaraDataInterfaceSkeletalMesh
   defaultValue {
    struct=NiagaraExt_VariableValue_DataInterface
    value {dataInterface=&self:NiagaraDataInterfaceSkeletalMesh_0}
   }
  }]
}
