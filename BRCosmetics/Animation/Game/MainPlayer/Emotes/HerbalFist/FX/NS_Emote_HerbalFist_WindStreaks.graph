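# Niagara system for the HerbalFist emote that draws faint, pale blue-white wind-streak ribbons trailing the character's back and both feet.
# Three anim-trail emitters use shared FN_ modules to spawn, initialize and update ribbons from bone or socket index pairs.
# Foot trails are much thinner and shorter-lived than the back trail.
# Each update sets color from an age-fade curve scaled by time of day and shrinks ribbon width.
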NiagaraSystem /BRCosmetics/Animation/Game/MainPlayer/Emotes/HerbalFist/FX/NS_Emote_HerbalFist_WindStreaks
dependencies {
 usedDataInterfaces [&NiagaraDataInterfaceCurve, &NiagaraDataInterfaceSocketReader]
 usedDynamicInputs []
 usedModules [
  &SystemState
  &self:SystemSpawnScript.SystemScriptSource.SystemScriptGraph.NiagaraNodeAssignment_0.SetVariables_#1
 ]
}
emitters {
 AnimTrail_Back {
  emitterSpawnScript [&FN_RibbonIDOffsetUtility()]
  emitterUpdateScript [
   &EmitterState()
   &FN_SpawnAnimTrail()
  ]
  particleSpawnScript [
   &FN_InitializeAnimTrailBoneIndecies()
   &FN_InitializeAnimTrail(AnimTrailWidth=3.5, Color=<rgba 1 1 1 0.1>, Lifetime=0.55, Position Offset=<0 0 2>)
  ]
  particleUpdateScript [
   &SolveForcesAndVelocity(Clamp Velocity=false, Limit Acceleration=false)
   &FN_UpdateAnimTrail() {module=FN_UpdateAnimTrail001}
   /SetVariables(
    Particles.Color6:LinearColor=&Multiply_LinearColorByFloat(
     Linear Color=<rgba 0.8229 0.8853 1 0.025>
     Float=&Multiply_Float(
      A=&FloatFromCurve(
       FloatCurve=DataInterface() {propertyValues {Curve {keys [<at 0 1 RCIM_Cubic>, {arriveTangentWeight=0.9023, interpMode=RCIM_Cubic, tangentMode=RCTM_User, time=1, value=0}]}}}
       Scale Curve=&FNTimeOfDayScalar(Night Scale=1.55, Morning Scale=1.8, Day Scale=1.8, Evening Scale=1.8)
      )
      B=1
     )
    )
    Particles.RibbonWidth6:NiagaraFloat=&Multiply_Float(A=$Particles.RibbonWidth, B=0.95)
   ) {isSetParameters=true, module=SetVariables_#2}
  ]
  rendererClasses [&NiagaraRibbonRendererProperties]
  renderers [{rendererClass=&NiagaraRibbonRendererProperties}]
 }
 AnimTrail_Foot_L {
  emitterSpawnScript [&FN_RibbonIDOffsetUtility()]
  emitterUpdateScript [
   &EmitterState()
   &FN_SpawnAnimTrail()
  ]
  particleSpawnScript [
   &FN_InitializeAnimTrailBoneIndecies(First Bone or Socket Index=5, Second Bone or Socket Index=6)
   &FN_InitializeAnimTrail(AnimTrailWidth=0.2, Color=<rgba 1 1 1 0.1>, Lifetime=0.25, Position Offset=<0 0 2>)
  ]
  particleUpdateScript [
   &SolveForcesAndVelocity(Clamp Velocity=false, Limit Acceleration=false)
   &FN_UpdateAnimTrail() {module=FN_UpdateAnimTrail001}
   /SetVariables(
    Particles.Color6:LinearColor=&Multiply_LinearColorByFloat(
     Linear Color=<rgba 0.8229 0.8853 1 0.025>
     Float=&Multiply_Float(
      A=&FloatFromCurve(
       FloatCurve=DataInterface() {propertyValues {Curve {keys [<at 0 1 RCIM_Cubic>, {arriveTangent=-2.5155, interpMode=RCIM_Cubic, leaveTangent=-2.5155, tangentMode=RCTM_User, time=1, value=0}]}}}
       Scale Curve=&FNTimeOfDayScalar(Night Scale=1.55, Morning Scale=1.8, Day Scale=1.8, Evening Scale=1.8)
      )
      B=1
     )
    )
    Particles.RibbonWidth6:NiagaraFloat=&Multiply_Float(A=$Particles.RibbonWidth, B=0.95)
   ) {isSetParameters=true, module=SetVariables_#2}
  ]
  rendererClasses [&NiagaraRibbonRendererProperties]
  renderers [{rendererClass=&NiagaraRibbonRendererProperties}]
 }
 AnimTrail_Foot_R {
  emitterSpawnScript [&FN_RibbonIDOffsetUtility()]
  emitterUpdateScript [
   &EmitterState()
   &FN_SpawnAnimTrail()
  ]
  particleSpawnScript [
   &FN_InitializeAnimTrailBoneIndecies(First Bone or Socket Index=3, Second Bone or Socket Index=4)
   &FN_InitializeAnimTrail(AnimTrailWidth=0.2, Color=<rgba 1 1 1 0.1>, Lifetime=0.25, Position Offset=<0 0 2>)
  ]
  particleUpdateScript [
   &SolveForcesAndVelocity(Clamp Velocity=false, Limit Acceleration=false)
   &FN_UpdateAnimTrail() {module=FN_UpdateAnimTrail001}
   /SetVariables(
    Particles.Color6:LinearColor=&Multiply_LinearColorByFloat(
     Linear Color=<rgba 0.8229 0.8853 1 0.025>
     Float=&Multiply_Float(
      A=&FloatFromCurve(
       FloatCurve=DataInterface() {propertyValues {Curve {keys [<at 0 1 RCIM_Cubic>, {arriveTangent=-2.5155, interpMode=RCIM_Cubic, leaveTangent=-2.5155, tangentMode=RCTM_User, time=1, value=0}]}}}
       Scale Curve=&FNTimeOfDayScalar(Night Scale=1.55, Morning Scale=1.8, Day Scale=1.8, Evening Scale=1.8)
      )
      B=1
     )
    )
    Particles.RibbonWidth6:NiagaraFloat=&Multiply_Float(A=$Particles.RibbonWidth, B=0.95)
   ) {isSetParameters=true, module=SetVariables_#2}
  ]
  rendererClasses [&NiagaraRibbonRendererProperties]
  renderers [{rendererClass=&NiagaraRibbonRendererProperties}]
 }
}
summary {
 systemName=NS_Emote_HerbalFist_WindStreaks
 emitters [
  {emitterName=AnimTrail_Back, rendererClasses [&NiagaraRibbonRendererProperties]}
  {emitterName=AnimTrail_Foot_R, rendererClasses [&NiagaraRibbonRendererProperties]}
  {emitterName=AnimTrail_Foot_L, rendererClasses [&NiagaraRibbonRendererProperties]}
 ]
 userVariables [
  User.bIsFrontEnd {type=NiagaraBool}
  User.bIsFrontEndPreview {type=NiagaraBool}
  User.BlinkIndex {type=NiagaraFloat, defaultValue=0, description=""}
  User.BlinkWidth {type=NiagaraFloat, defaultValue=1, description=""}
  User.RibbonIDOffset {type=NiagaraInt32}
 ]
}
